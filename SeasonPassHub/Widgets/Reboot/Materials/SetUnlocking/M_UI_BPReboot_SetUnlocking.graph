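# Additive UI material for the Season Pass hub, apparently a set-unlocking effect driven by a Reveal scalar.
# Polar-mapped, panning noise forms swirling highlights and a two-color rift gradient, with distance-field circles shaping glow and pulse ring.
# A scaled grid with per-cell randomness scatters dust dots.
# Named reroutes feed these layers into lerp and color-dodge compositing, alongside a rounded-corner vertical-fade mask and optional tile mask.
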Material /SeasonPassHub/Widgets/Reboot/Materials/SetUnlocking/M_UI_BPReboot_SetUnlocking
graph {
 properties {
  bAutomaticallySetUsageInEditor=false
  bUsedWithStaticMesh=false
  blendMode=BLEND_Additive
  materialDomain=MD_UI
  usedShadingModels=Default Lit
 }
 nodes [
  Frac(None=Multiply(A=Time()) {constB=0.5})
  n1 = NamedRerouteUsage()
  n2 = Multiply(A=LinearInterpolate(A=NamedRerouteUsage(), Alpha=Power(Base=@n1)), B=@n1)
  n3 = NamedRerouteUsage()
  n4 = NamedRerouteUsage()
  n5 = Constant()
  n6 = Power(Base=OneMinus(None=MaterialFunctionCall(UVs=@n4, Size=@n5) {materialFunction=&MF_UI_Circle}.SDF)) {constExponent=6}
  LinearInterpolate(
   A=MaterialFunctionCall!(Blend=Min(A=@n2) {constB=0.8}, Base=LinearInterpolate(A=@n3, Alpha=@n6)) {materialFunction=/Engine/Functions/Engine_MaterialFunctions03/Blends/Blend_ColorDodge}.Result
   Alpha=NamedRerouteUsage()
  )
  n7 = LinearInterpolate(A=NamedRerouteUsage(), Alpha=LinearInterpolate(A=NamedRerouteUsage(), Alpha=NamedRerouteUsage()))
  LinearInterpolate(A=@n7, B=Multiply(A=@n7, B=NamedRerouteUsage()), Alpha=ScalarParameter() {parameterName=UseTileMask})
  n8 = NamedRerouteUsage()
  n9 = SmoothStep(Value=NamedRerouteUsage()) {constMax=0.8}
  n10 = Constant() {r=0.5}
  n11 = OneMinus(None=Multiply(A=Distance(A=@n8, B=@n10)) {constB=2})
  n12 = MaterialFunctionCall(UVs=@n8, Texture Scale=Power(Base=Power(Base=@n9) {constExponent=0.5}, Exp=Power(Base=@n11) {constExponent=3})) {materialFunction=&ScaleUVsByCenter}
  n13 = Constant() {r=30}
  n14 = MaterialFunctionCall!(
   UV=MaterialFunctionCall(UV=NamedRerouteDeclaration(None=@n12.UVs) {name=UV, nodeColor=<rgba 0 0 0 1>}) {materialFunction=&MF_UI_UVScaling}.Scaled to Fit
   Tiling X=@n13
   Tiling Y=@n13
  ) {materialFunction=/FortniteCoreUI/UI/Community/Art/MaterialFunctions/MF_UI_GridTiling}
  NamedRerouteDeclaration(None=@n14.PatternUVs) {name=TiledUV, nodeColor=<rgba 0 0 0 1>}
  NamedRerouteDeclaration(None=Divide(A=Floor(None=@n14.Raw UVs), B=@n13)) {name=SteppedUV, nodeColor=<rgba 0 0 0 1>}
  NamedRerouteDeclaration(None=OneMinus(None=ScalarParameter() {defaultValue=0.5, parameterName=Reveal})) {name=Reveal, nodeColor=<rgba 0 0 0 1>}
  n15 = MaterialFunctionCall(UV=NamedRerouteUsage()) {materialFunction=&MF_UI_UVScaling}
  n16 = MaterialFunctionCall!(UV Coord=@n15.Scaled to Fit, LenghtScale, RadialScale) {materialFunction=/Game/Global/MaterialFunctions/TechArt/MF_PolarCoordinates}
  n17 = Constant2Vector() {g=2, r=3}
  n18 = TextureSample!(
   UVs=Panner(Coordinate=MaterialFunctionCall(UVs=@n16.Result, Texture Scale=@n17) {materialFunction=&ScaleUVsByCenter}.UVs) {speedX=0.1, speedY=0.25}
   Tex
   "DDX(UVs)"=DDX(Value=@n15.Scaled to Fit)
   "DDY(UVs)"=DDY(Value=@n15.Scaled to Fit)
   Apply View MipBias
  ) {mipValueMode=TMVM_Derivative, samplerType=SAMPLERTYPE_Grayscale, texture=&LowResBlurredNoise}
  n19 = MaterialFunctionCall(UV=NamedRerouteUsage()) {materialFunction=&MF_UI_UVScaling}
  n20 = MaterialFunctionCall(UV Coord=@n19.Scaled to Fit) {materialFunction=&MF_PolarCoordinates}
  n21 = AppendVector(A=ComponentMask(None=@n20.Result) {g=true}, B=ComponentMask(None=@n20.Result) {r=true})
  n22 = Constant2Vector() {g=5, r=0.5}
  n23 = TextureSample(
   UVs=Panner(Coordinate=MaterialFunctionCall(UVs=@n21, Texture Scale=@n22) {materialFunction=&ScaleUVsByCenter}.UVs) {speedX=-0.1, speedY=0.1}
   "DDX(UVs)"=DDX(Value=@n19.Scaled to Fit)
   "DDY(UVs)"=DDY(Value=@n19.Scaled to Fit)
  ) {mipValueMode=TMVM_Derivative, samplerType=SAMPLERTYPE_Grayscale, texture=&LowResBlurredNoise}
  n24 = LinearInterpolate(A=@n18.R, Alpha=Power(Base=@n23.R))
  n25 = SmoothStep(Value=NamedRerouteUsage())
  n26 = OneMinus(None=@n25)
  n27 = Power(Base=Multiply(A=Min(A=@n25, B=@n26)) {constB=2}) {constExponent=0.2}
  n28 = OneMinus(None=MaterialFunctionCall(UVs=NamedRerouteUsage(), Size=LinearInterpolate(Alpha=@n27) {constB=0.2}) {materialFunction=&MF_UI_Circle}.SDF)
  n29 = LinearInterpolate(
   A=Multiply(A=NamedRerouteDeclaration(None=@n24) {name=Highlights, nodeColor=<rgba 0 0 0 1>}, B=@n28) {constB=0.5}
   Alpha=Power(Base=@n28) {constExponent=3}
  )
  NamedRerouteDeclaration(None=Multiply(A=Power(Base=@n29), B=@n27)) {name=Glow Shape, nodeColor=<rgba 0 0 0 1>}
  n30 = MaterialFunctionCall(UV=NamedRerouteUsage()) {materialFunction=&MF_UI_UVScaling}
  n31 = SmoothStep(Value=Power(Base=NamedRerouteUsage()) {constExponent=0.5}) {constMax=0.8}
  n32 = MaterialFunctionCall(
   UVs=MaterialFunctionCall(UVs=@n30.Scaled to Fit, Texture Scale=LinearInterpolate(Alpha=@n31) {constB=6}) {materialFunction=&ScaleUVsByCenter}.UVs
   Size=Constant() {r=0.5}
  ) {materialFunction=&MF_UI_Circle}
  n33 = Constant() {r=0.5}
  n34 = OneMinus(None=Multiply(A=Distance(A=@n32.SDF, B=@n33)) {constB=2})
  n35 = MaterialFunctionCall(UVs=@n30.Scaled to Fit, Size=Constant()) {materialFunction=&MF_UI_Circle}
  NamedRerouteDeclaration(None=Multiply(A=Saturate(None=@n34), B=OneMinus(None=@n35.SDF))) {name=Pulse, nodeColor=<rgba 0 0 0 1>}
  n36 = NamedRerouteUsage()
  n37 = LinearInterpolate(
   A=Multiply(A=NamedRerouteUsage()) {constB=-1}
   Alpha=MaterialFunctionCall!(UV=NamedRerouteUsage(), Seed) {materialFunction=&MF_UI_Random}.Result
  ) {constA=-1, constB=0}
  n38 = NamedRerouteUsage()
  n39 = Constant() {r=2}
  n40 = Multiply(
   A=OneMinus(None=MaterialFunctionCall(UVs=@n36, Size=@n37) {materialFunction=&MF_UI_Circle}.SDF)
   B=Step(Y=MaterialFunctionCall(UV=@n38, Seed=@n39) {materialFunction=&MF_UI_Random}.Result) {constX=0.2}
  )
  n41 = NamedRerouteUsage()
  n42 = Constant() {r=1}
  n43 = Multiply(
   A=Multiply(
    A=Power(Base=@n40)
    B=MaterialFunctionCall(UV=@n41, Seed=@n42) {materialFunction=&MF_UI_Random}.Result
   )
   B=Power(Base=NamedRerouteUsage()) {constExponent=0.5}
  )
  n44 = MaterialFunctionCall(DistanceField=@n43, Threshold=Constant() {r=0.2}, Smoothing=Constant() {r=0.2}) {materialFunction=&MF_UI_AntiAliasing}
  n45 = NamedRerouteUsage()
  n46 = Min(A=@n45, B=OneMinus(None=@n45))
  NamedRerouteDeclaration(None=Multiply(A=LinearInterpolate(A=@n43, Alpha=@n44.Outer), B=Multiply(A=@n46) {constB=2})) {name=Dust, nodeColor=<rgba 0 0 0 1>}
  n47 = MaterialFunctionCall(UV=NamedRerouteUsage()) {materialFunction=&MF_UI_UVScaling}
  n48 = MaterialFunctionCall(UV Coord=@n47.Scaled to Fit) {materialFunction=&MF_PolarCoordinates}
  n49 = Constant2Vector() {g=2, r=4}
  n50 = TextureSample(
   UVs=Panner(Coordinate=MaterialFunctionCall(UVs=@n48.Result, Texture Scale=@n49) {materialFunction=&ScaleUVsByCenter}.UVs) {speedX=0.1, speedY=0.25}
   "DDX(UVs)"=DDX(Value=@n47.Scaled to Fit)
   "DDY(UVs)"=DDY(Value=@n47.Scaled to Fit)
  ) {mipValueMode=TMVM_Derivative, samplerType=SAMPLERTYPE_Grayscale, texture=&LowResBlurredNoise}
  NamedRerouteDeclaration(None=LinearInterpolate(
    A=VectorParameter() {group=0 - Coloring, parameterName=Color 1, defaultValue=<rgba 0 0.1426 0.7917 1>}.RGB
    B=VectorParameter() {group=0 - Coloring, parameterName=Color 2, defaultValue=<rgba 0.3365 0.0119 0.7604 1>}.RGB
    Alpha=Power(Base=@n50.R) {constExponent=0.5}
   )) {name=Rift Gradient, nodeColor=<rgba 0 0 0 1>}
  n51 = MaterialFunctionCall() {materialFunction=&GetUserInterfaceUV}
  n52 = AppendVector(A=ScalarParameter() {parameterName=Pos X}, B=ScalarParameter() {parameterName=Pos Y})
  NamedRerouteDeclaration(None=MaterialFunctionCall(UVs=Add(A=@n51.Normalized UV, B=@n52), Texture Scale=ScalarParameter() {defaultValue=1, parameterName=Scale}) {materialFunction=&ScaleUVsByCenter}.UVs) {name=UV Position, nodeColor=<rgba 1 0 0.9141 1>}
  n53 = MaterialFunctionCall(
   R=ScalarParameter() {defaultValue=10, group=Layout, parameterName=TopLeftCornerRadius, sortPriority=4}
   G=ScalarParameter() {defaultValue=10, group=Layout, parameterName=BottomLeftCornerRadius, sortPriority=7}
   B=ScalarParameter() {defaultValue=10, group=Layout, parameterName=TopRightCornerRadius, sortPriority=5}
   A=ScalarParameter() {defaultValue=10, group=Layout, parameterName=BottomRightCornerRadius, sortPriority=6}
  ) {materialFunction=&AppendMany}
  n54 = ScalarParameter() {
   defaultValue=1
   desc="If enabled reads corner radius values in pixels.\r\nbool"
   group=Layout
   parameterName=UseValueInPixels
   sliderMax=1
   sortPriority=1
  }
  n55 = MaterialFunctionCall() {materialFunction=&GetUserInterfaceUV}
  NamedRerouteDeclaration(None=Multiply(
    A=MaterialFunctionCall(Corner Radius=@n53.RGBA, Radius in Pixels=@n54) {materialFunction=&MF_UI_CoreUI_RoundedMask}.Shape Mask
    B=ComponentMask(None=@n55.Normalized UV) {g=true}
   )) {name=Shape Mask, nodeColor=<rgba 0 0 0 1>}
  ScalarParameter() {defaultValue=20, parameterName=TileAmount}
 ]
}
